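; Balloon actor for the Pop-A-Loon quest interactable that inflates and can be popped.
; FillBalloon shows the mesh and plays the Inflate timeline and sound, using the inverse of the board-supplied FillSpeed as play rate.
; BurstBalloon clears bActive, whose RepNotify hides the mesh, resets the timeline, fires pop effects and fades the inflate sound.
; KillBalloons hides and resets the balloon without pop effects.
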(event EventKillBalloons
  (bind _inflate (Variables|BP_PoppingBalloon|GetInflate))
  (Development|SetHiddeninGame (Variables|Default|GetSM_Balloon) true false)
  (Components|Timeline|Stop _inflate)
  (Components|Timeline|SetNewTime _inflate 0.0)
  (Audio|Components|Audio|Stop (Variables|Default|GetBalloon_Quest_Inflate_Sound)))

(event EventFillBalloon
  (params (FillSpeed :type Float%20%28double-precision%29))
  (bind _inflate (Variables|BP_PoppingBalloon|GetInflate))
  ;[0.6217 1 0.4846] Controls speed of playback, set by board on spawn of balloon
  (Variables|Default|SetbActive true)
  (Variables|Default|SetLocalFillSpeed (/ 1.0 FillSpeed))
  (Development|SetHiddeninGame (Variables|Default|GetSM_Balloon) false false)
  (Components|Timeline|SetPlayRate _inflate (Variables|Default|GetLocalFillSpeed))
  (Components|Timeline|PlayfromStart _inflate)
  (Audio|Components|Audio|Play (Variables|Default|GetBalloon_Quest_Inflate_Sound) 0.0))

(event EventBurstBalloon
  ;[1 0.3126 0] Got hit, need to shutdown
  (Variables|Default|SetbActive false))

(fn OnRep_bActive
  (bind _p_balloonboard_pop (Variables|Default|GetP_BalloonBoard_Pop))
  (bind _inflate (Variables|BP_PoppingBalloon|GetInflate))
  (if (not (Variables|Default|GetbActive))
    (Development|SetHiddeninGame (Variables|Default|GetSM_Balloon) true false)
    (Components|Timeline|Stop _inflate)
    (Components|Timeline|SetNewTime _inflate 0.0)
    (Components|Activation|Activate _p_balloonboard_pop true)
    ;[1 0 0.8711] AUDIO - Balloons have been destroyed here
    (Audio|PlaySoundatLocation "/Figment_S10/Sounds/Fort_Gadgets/Balloons/Balloons_Quest_Pop_Cue.Balloons_Quest_Pop_Cue" (Transformation|GetWorldLocation _p_balloonboard_pop))
    (Audio|Components|Audio|FadeOut (Variables|Default|GetBalloon_Quest_Inflate_Sound) 0.2 0.0 "Linear" :!class /Script/Engine.AudioComponent)))

(fn OnRep_bGotHit)

(fn OnRep_BoardOwner)

(fn OnRep_FillSpeed)
